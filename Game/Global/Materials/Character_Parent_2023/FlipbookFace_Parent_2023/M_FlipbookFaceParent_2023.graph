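# Parent character material for flipbook faces, with per-pixel shading models and support for skeletal, cloth, morph, Nanite and Niagara meshes.
# Material attributes pass from the character PreFX function through MF_FlipbookFace, which handles mouth, eye and brow, into the PostFX function.
# Face alpha times HasPoseChanged, a parameter the head AnimBP sets, raises temporal responsiveness on face pixels when the pose changes.
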Material /Game/Global/Materials/Character_Parent_2023/FlipbookFace_Parent_2023/M_FlipbookFaceParent_2023
graph {
 properties {
  bUseMaterialAttributes=true
  bUsedWithClothing=true
  bUsedWithMeshDeformer=true
  bUsedWithMorphTargets=true
  bUsedWithNanite=true
  bUsedWithNiagaraMeshParticles=true
  bUsedWithSkeletalMesh=true
  pixelDepthOffsetMode=PDOM_Legacy
  shadingModel=MSM_FromMaterialExpression
  usedShadingModels=Default Lit | Subsurface | Clear Coat | Subsurface Profile | Cloth
 }
 nodes [
  n1 = MaterialFunctionCall() {materialFunction=&MF_CharacterParent_2023_PreFX}
  n2 = MaterialFunctionCall!(Mouth UV, Eye UV, Brow UV, Material=@n1.Result) {materialFunction=/Game/Global/Materials/Character_Parent_2023/FlipbookFace_Parent_2023/MaterialFunctions/MF_FlipbookFace}
  MaterialFunctionCall(Material=@n2.Material, M=@n1.M, Pre Skinned Local Position=@n1.M, Distance Blend=@n1.M, Fresnel=@n1.M) {materialFunction=&MF_CharacterParent_2023_PostFX}
  n3 = ScalarParameter() {desc="Set by base head AnimBP if \"IsFlipbookFace\" is true", parameterName=HasPoseChanged}
  TemporalResponsivenessOutput!(None=LinearInterpolate(Alpha=Multiply(A=@n2.FaceAlpha, B=@n3)) {constA=0.4})
 ]
}
